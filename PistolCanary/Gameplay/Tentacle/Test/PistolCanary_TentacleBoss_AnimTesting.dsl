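; Test Blueprint for previewing single-tentacle animations of the PistolCanary tentacle boss.
; Each named function calls PlayAnimation on the skeletal mesh with one hard-coded asset, covering attacks, hit reacts, death, idle, MaskMaker hit and 180-degree turns.
; Only the two turn idle functions loop, and the turn clips come from an Unused folder.
; No sequencing, state or construction logic is present.
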(fn CircularBeam
  (Components|Animation|PlayAnimation (Variables|Default|GetSkeletalMesh) "/PistolCanary/Anim/SingleTentacle/Section02/PistolCanary_Attack_CircularBeam/PistolCanary_Tentacle_Attack_CircularBeam.PistolCanary_Tentacle_Attack_CircularBeam" false))

(fn Death
  (Components|Animation|PlayAnimation (Variables|Default|GetSkeletalMesh) "/PistolCanary/Anim/SingleTentacle/Section03/PistolCanary_Tentacle_Death_01.PistolCanary_Tentacle_Death_01" false))

(fn EnergyBlast
  (Components|Animation|PlayAnimation (Variables|Default|GetSkeletalMesh) "/PistolCanary/Anim/SingleTentacle/Section02/PistolCanary_Attack_EnergyBlast/PistolCanary_Tentacle_Attack_EnergyBlast.PistolCanary_Tentacle_Attack_EnergyBlast" false))

(fn EnergyWhip
  (Components|Animation|PlayAnimation (Variables|Default|GetSkeletalMesh) "/PistolCanary/Anim/SingleTentacle/Section02/PistolCanary_Attack_EnergyWhip/PistolCanary_Tentacle_Attack_EnergyWhip.PistolCanary_Tentacle_Attack_EnergyWhip" false))

(fn GroundSlam
  (Components|Animation|PlayAnimation (Variables|Default|GetSkeletalMesh) "/PistolCanary/Anim/SingleTentacle/Section02/PistolCanary_Attack_GroundSlam/PistolCanary_Tentacle_Attack_GroundSlam.PistolCanary_Tentacle_Attack_GroundSlam" false))

(fn GroundSlamJab
  (Components|Animation|PlayAnimation (Variables|Default|GetSkeletalMesh) "/PistolCanary/Anim/SingleTentacle/Section02/PistolCanary_Attack_GroundSlamJab/PistolCanary_Tentacle_Attack_GroundSlamJab.PistolCanary_Tentacle_Attack_GroundSlamJab" false))

(fn HitReactA
  (Components|Animation|PlayAnimation (Variables|Default|GetSkeletalMesh) "/PistolCanary/Anim/SingleTentacle/Section02/PistolCanary_Tentacle_HitReacts/PistolCanary_Tentacle_Damage_Thrashing.PistolCanary_Tentacle_Damage_Thrashing" false))

(fn HitReactB
  (Components|Animation|PlayAnimation (Variables|Default|GetSkeletalMesh) "/PistolCanary/Anim/SingleTentacle/Section02/PistolCanary_Tentacle_HitReacts/PistolCanary_Tentacle_Damage_Thrashing_02.PistolCanary_Tentacle_Damage_Thrashing_02" false))

(fn HomingBurst
  (Components|Animation|PlayAnimation (Variables|Default|GetSkeletalMesh) "/PistolCanary/Anim/SingleTentacle/Section02/PistolCanary_Attack_HomingBurst/PistolCanary_Tentacle_Attack_HomingBurst.PistolCanary_Tentacle_Attack_HomingBurst" false))

(fn Idle
  (Components|Animation|PlayAnimation (Variables|Default|GetSkeletalMesh) "/PistolCanary/Anim/SingleTentacle/Section02/PistolCanary_Tentacle_Idles/PistolCanary_Tentacle_Idle_01.PistolCanary_Tentacle_Idle_01" false))

(fn MaskMakerHit
  (Components|Animation|PlayAnimation (Variables|Default|GetSkeletalMesh) "/PistolCanary/Anim/SingleTentacle/Section01/PistolCanary_Tentacle_MaskMakerHit.PistolCanary_Tentacle_MaskMakerHit" false))

(fn MaskMakerHit_Idle
  (Components|Animation|PlayAnimation (Variables|Default|GetSkeletalMesh) "/PistolCanary/Anim/SingleTentacle/Section01/PistolCanary_Tentacle_MaskMakerHit_Idle.PistolCanary_Tentacle_MaskMakerHit_Idle" false))

(fn MaskMakerHit_Intro
  (Components|Animation|PlayAnimation (Variables|Default|GetSkeletalMesh) "/PistolCanary/Anim/SingleTentacle/Section01/PistolCanary_Tentacle_MaskMakerHit_Intro.PistolCanary_Tentacle_MaskMakerHit_Intro" false))

(fn Turn180L
  (Components|Animation|PlayAnimation (Variables|Default|GetSkeletalMesh) "/PistolCanary/Anim/SingleTentacle/Unused/PistolCanary_Tentacle_Turn_180L/PistolCanary_Tentacle_Turn_180L.PistolCanary_Tentacle_Turn_180L" false))

(fn Turn180L_Idle
  (Components|Animation|PlayAnimation (Variables|Default|GetSkeletalMesh) "/PistolCanary/Anim/SingleTentacle/Unused/PistolCanary_Tentacle_Turn_180L/PistolCanary_Tentacle_Turn_180L_Idle.PistolCanary_Tentacle_Turn_180L_Idle" true))

(fn Turn180L_Return
  (Components|Animation|PlayAnimation (Variables|Default|GetSkeletalMesh) "/PistolCanary/Anim/SingleTentacle/Unused/PistolCanary_Tentacle_Turn_180L/PistolCanary_Tentacle_Turn_180L_Return.PistolCanary_Tentacle_Turn_180L_Return" false))

(fn Turn180R
  (Components|Animation|PlayAnimation (Variables|Default|GetSkeletalMesh) "/PistolCanary/Anim/SingleTentacle/Unused/PistolCanary_Tentacle_Turn_180R/PistolCanary_Tentacle_Turn_180R.PistolCanary_Tentacle_Turn_180R" false))

(fn Turn180R_Idle
  (Components|Animation|PlayAnimation (Variables|Default|GetSkeletalMesh) "/PistolCanary/Anim/SingleTentacle/Unused/PistolCanary_Tentacle_Turn_180R/PistolCanary_Tentacle_Turn_180R_Idle.PistolCanary_Tentacle_Turn_180R_Idle" true))

(fn Turn180R_Return
  (Components|Animation|PlayAnimation (Variables|Default|GetSkeletalMesh) "/PistolCanary/Anim/SingleTentacle/Unused/PistolCanary_Tentacle_Turn_180R/PistolCanary_Tentacle_Turn_180R_Return.PistolCanary_Tentacle_Turn_180R_Return" false))

(fn ConstructionScript)
