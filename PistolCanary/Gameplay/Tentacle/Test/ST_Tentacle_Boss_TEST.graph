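# Test StateTree for a tentacle boss's combat AI, with a root Combat state that shows debug text and runs a Blueprint tentacle task.
# Select Ability holds an empty Idle and a Random group of five attack states, each gated by Can Activate Ability and starting its tagged ability.
# The Wait branch's Active state delays forever with a target-focus task, returning to Select Ability on tick once Gameplay.Status.TB.Active is absent.
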StateTree /PistolCanary/Gameplay/Tentacle/Test/ST_Tentacle_Boss_TEST
states [
 Combat {
  id=#1
  type=STATE
  children [#2, #3]
  tasks [
   {
    id=#4
    description="Debug Text \"COMBAT\""
    properties {
     iD=#5
     instance {_structType=/Script/StateTreeModule.StateTreeDebugTextTaskInstanceData, bEnabled=true, bindableText="", referenceActor=null}
     node {
      name=None
      _structType=/Script/StateTreeModule.StateTreeDebugTextTask
      bCanEditConsideredForCompletion=false
      bConsideredForCompletion=false
      bHasTaskCompletionDelegateDispatcher=false
      bTaskEnabled=true
      deprecatedVersion=1
      fontScale=1
      text=COMBAT
      transitionHandlingPriority=Normal
      bindingsBatch {value=65535}
      executionRuntimeTemplateIndex {value=65535}
      instanceDataHandle {index=65535, source=None, stateHandle {index=65535}}
      instanceTemplateIndex {value=65535}
      offset=<0 0 0>
      outputBindingsBatch {value=65535}
      textColor=<rgba 1 0.4824 0.0078 1>
     }
    }
   }
   {
    id=#6
    description=State Tree Task Tentacle
    properties {
     iD=#7
     instanceObject=&self:StateTreeEditorData_0.StateTreeState_0.StateTreeTask_Tentacle_C_0
     node {
      name=None
      _structType=&StateTreeBlueprintTaskWrapper
      taskClass=&/StateTreeTask_Tentacle.StateTreeTask_Tentacle_C
     }
    }
   }
  ]
  transitions []
 }
 Select Ability {
  id=#2
  type=STATE
  children [#8, #9]
  tasks []
  transitions []
 }
 Wait {id=#3, type=STATE, children [#10], tasks [], transitions []}
 Idle {id=#8, type=STATE, tasks [], transitions []}
 Random {
  id=#9
  type=STATE
  children [
   #11
   #12
   #13
   #14
   #15
  ]
  tasks []
  transitions []
 }
 Active {
  id=#10
  type=STATE
  enter_conditions [{
    id=#16
    description="(Fortnite) Actor Has Matching Gameplay Tag"
    properties {
     iD=#17
     instance {_structType=/Script/FortniteAIServer.GameplayTagActorHasMatchingGameplayTagInstanceData, actor=null, tagToCheck {tagName=None}}
     node {
      name=None
      _structType=/Script/FortniteAIServer.GameplayTagActorHasMatchingGameplayTagCondition
      bInvert=true
      deltaIndent=0
      evaluationMode=Evaluated
      operand=And
      bindingsBatch {value=65535}
      executionRuntimeTemplateIndex {value=65535}
      instanceDataHandle {index=65535, source=None, stateHandle {index=65535}}
      instanceTemplateIndex {value=65535}
      outputBindingsBatch {value=65535}
     }
    }
   }]
  tasks [
   {
    id=#18
    description=Delay Forever
    properties {
     iD=#19
     instance {_structType=&StateTreeDelayTaskInstanceData, bRunForever=true, duration=1}
     node {name=None, _structType=&StateTreeDelayTask}
    }
   }
   {
    id=#20
    description="[ActorFocus] Set Target Focus"
    properties {
     iD=#21
     instance {_structType=/Script/FortniteAI.FortActorFocusComponentSetActorFocusInstanceData, actor=null, focusActor=null}
     node {
      name=None
      _structType=/Script/FortniteAI.FortActorFocusComponentSetActorFocusTask
      bCanEditConsideredForCompletion=true
      bConsideredForCompletion=true
      bHasTaskCompletionDelegateDispatcher=false
      bTaskEnabled=true
      transitionHandlingPriority=Normal
      bindingsBatch {value=65535}
      executionRuntimeTemplateIndex {value=65535}
      instanceDataHandle {index=65535, source=None, stateHandle {index=65535}}
      instanceTemplateIndex {value=65535}
      outputBindingsBatch {value=65535}
     }
    }
   }
  ]
  transitions [{
    id=#22
    trigger=ON_TICK
    conditions [{
      id=#23
      description="(Fortnite) Actor Has Matching Gameplay Tag"
      properties {
       iD=#24
       instance {
        _structType=&GameplayTagActorHasMatchingGameplayTagInstanceData
        actor=null
        tagToCheck {tagName=Gameplay.Status.TB.Active}
       }
       node {
        name=None
        _structType=&GameplayTagActorHasMatchingGameplayTagCondition
        bInvert=true
        deltaIndent=0
        evaluationMode=Evaluated
        operand=And
        bindingsBatch {value=65535}
        executionRuntimeTemplateIndex {value=65535}
        instanceDataHandle {index=65535, source=None, stateHandle {index=65535}}
        instanceTemplateIndex {value=65535}
        outputBindingsBatch {value=65535}
       }
      }
     }]
    state {id=#2, name=Select Ability}
   }]
 }
 PewPewPew {
  id=#11
  type=STATE
  enter_conditions [{
    id=#25
    description=Can Activate Ability Ability.TB.EnergyBlast
    properties {
     iD=#26
     instance {
      _structType=/Script/FortniteAIServer.FortGameplayAbilityCanActivateAbilityInstanceData
      pawn=null
      abilityTag {gameplayTags [{tagName=Ability.TB.EnergyBlast}]}
     }
     node {
      name=None
      _structType=/Script/FortniteAIServer.FortGameplayAbilityCanActivateAbilityCondition
      bInvert=false
      deltaIndent=0
      evaluationMode=Evaluated
      operand=And
      bindingsBatch {value=65535}
      executionRuntimeTemplateIndex {value=65535}
      instanceDataHandle {index=65535, source=None, stateHandle {index=65535}}
      instanceTemplateIndex {value=65535}
      outputBindingsBatch {value=65535}
     }
    }
   }]
  tasks [{
    id=#27
    description=Start Ability Ability.TB.EnergyBlast
    properties {
     iD=#28
     instance {
      _structType=/Script/FortniteGame.FortStateTreeStartAbilityTaskInstanceData
      actor=null
      abilityTag {gameplayTags [{tagName=Ability.TB.EnergyBlast}]}
     }
     node {
      name=None
      _structType=/Script/FortniteGame.FortStateTreeStartAbilityTask
      bCanEditConsideredForCompletion=true
      bConsideredForCompletion=true
      bHasTaskCompletionDelegateDispatcher=false
      bTaskEnabled=true
      transitionHandlingPriority=Normal
      bindingsBatch {value=65535}
      executionRuntimeTemplateIndex {value=65535}
      instanceDataHandle {index=65535, source=None, stateHandle {index=65535}}
      instanceTemplateIndex {value=65535}
      outputBindingsBatch {value=65535}
     }
    }
   }]
  transitions []
 }
 Circular Beam {
  id=#12
  type=STATE
  enter_conditions [{
    id=#29
    description=Can Activate Ability Ability.TB.CircularBeam
    properties {
     iD=#30
     instance {
      _structType=&FortGameplayAbilityCanActivateAbilityInstanceData
      pawn=null
      abilityTag {gameplayTags [{tagName=Ability.TB.CircularBeam}]}
     }
     node {
      name=None
      _structType=&FortGameplayAbilityCanActivateAbilityCondition
      bInvert=false
      deltaIndent=0
      evaluationMode=Evaluated
      operand=And
      bindingsBatch {value=65535}
      executionRuntimeTemplateIndex {value=65535}
      instanceDataHandle {index=65535, source=None, stateHandle {index=65535}}
      instanceTemplateIndex {value=65535}
      outputBindingsBatch {value=65535}
     }
    }
   }]
  tasks [{
    id=#31
    description=Start Ability Ability.TB.CircularBeam
    properties {
     iD=#32
     instance {
      _structType=&FortStateTreeStartAbilityTaskInstanceData
      actor=null
      abilityTag {gameplayTags [{tagName=Ability.TB.CircularBeam}]}
     }
     node {
      name=None
      _structType=&FortStateTreeStartAbilityTask
      bCanEditConsideredForCompletion=true
      bConsideredForCompletion=true
      bHasTaskCompletionDelegateDispatcher=false
      bTaskEnabled=true
      transitionHandlingPriority=Normal
      bindingsBatch {value=65535}
      executionRuntimeTemplateIndex {value=65535}
      instanceDataHandle {index=65535, source=None, stateHandle {index=65535}}
      instanceTemplateIndex {value=65535}
      outputBindingsBatch {value=65535}
     }
    }
   }]
  transitions []
 }
 Homing Blast {
  id=#13
  type=STATE
  enter_conditions [{
    id=#33
    description=Can Activate Ability Ability.TB.HomingBlast
    properties {
     iD=#34
     instance {
      _structType=&FortGameplayAbilityCanActivateAbilityInstanceData
      pawn=null
      abilityTag {gameplayTags [{tagName=Ability.TB.HomingBlast}]}
     }
     node {
      name=None
      _structType=&FortGameplayAbilityCanActivateAbilityCondition
      bInvert=false
      deltaIndent=0
      evaluationMode=Evaluated
      operand=And
      bindingsBatch {value=65535}
      executionRuntimeTemplateIndex {value=65535}
      instanceDataHandle {index=65535, source=None, stateHandle {index=65535}}
      instanceTemplateIndex {value=65535}
      outputBindingsBatch {value=65535}
     }
    }
   }]
  tasks [{
    id=#35
    description=Start Ability Ability.TB.HomingBlast
    properties {
     iD=#36
     instance {
      _structType=&FortStateTreeStartAbilityTaskInstanceData
      actor=null
      abilityTag {gameplayTags [{tagName=Ability.TB.HomingBlast}]}
     }
     node {
      name=None
      _structType=&FortStateTreeStartAbilityTask
      bCanEditConsideredForCompletion=true
      bConsideredForCompletion=true
      bHasTaskCompletionDelegateDispatcher=false
      bTaskEnabled=true
      transitionHandlingPriority=Normal
      bindingsBatch {value=65535}
      executionRuntimeTemplateIndex {value=65535}
      instanceDataHandle {index=65535, source=None, stateHandle {index=65535}}
      instanceTemplateIndex {value=65535}
      outputBindingsBatch {value=65535}
     }
    }
   }]
  transitions []
 }
 Ground Slam {
  id=#14
  type=STATE
  enter_conditions [{
    id=#37
    description=Can Activate Ability Ability.TB.GroundSlam
    properties {
     iD=#38
     instance {
      _structType=&FortGameplayAbilityCanActivateAbilityInstanceData
      pawn=null
      abilityTag {gameplayTags [{tagName=Ability.TB.GroundSlam}]}
     }
     node {
      name=None
      _structType=&FortGameplayAbilityCanActivateAbilityCondition
      bInvert=false
      deltaIndent=0
      evaluationMode=Evaluated
      operand=And
      bindingsBatch {value=65535}
      executionRuntimeTemplateIndex {value=65535}
      instanceDataHandle {index=65535, source=None, stateHandle {index=65535}}
      instanceTemplateIndex {value=65535}
      outputBindingsBatch {value=65535}
     }
    }
   }]
  tasks [{
    id=#39
    description=Start Ability Ability.TB.GroundSlam
    properties {
     iD=#40
     instance {
      _structType=&FortStateTreeStartAbilityTaskInstanceData
      actor=null
      abilityTag {gameplayTags [{tagName=Ability.TB.GroundSlam}]}
     }
     node {
      name=None
      _structType=&FortStateTreeStartAbilityTask
      bCanEditConsideredForCompletion=true
      bConsideredForCompletion=true
      bHasTaskCompletionDelegateDispatcher=false
      bTaskEnabled=true
      transitionHandlingPriority=Normal
      bindingsBatch {value=65535}
      executionRuntimeTemplateIndex {value=65535}
      instanceDataHandle {index=65535, source=None, stateHandle {index=65535}}
      instanceTemplateIndex {value=65535}
      outputBindingsBatch {value=65535}
     }
    }
   }]
  transitions []
 }
 Eye Lasers {
  id=#15
  type=STATE
  enter_conditions [{
    id=#41
    description=Can Activate Ability Ability.TB.Lasers
    properties {
     iD=#42
     instance {
      _structType=&FortGameplayAbilityCanActivateAbilityInstanceData
      pawn=null
      abilityTag {gameplayTags [{tagName=Ability.TB.Lasers}]}
     }
     node {
      name=None
      _structType=&FortGameplayAbilityCanActivateAbilityCondition
      bInvert=false
      deltaIndent=0
      evaluationMode=Evaluated
      operand=And
      bindingsBatch {value=65535}
      executionRuntimeTemplateIndex {value=65535}
      instanceDataHandle {index=65535, source=None, stateHandle {index=65535}}
      instanceTemplateIndex {value=65535}
      outputBindingsBatch {value=65535}
     }
    }
   }]
  tasks [{
    id=#43
    description=Start Ability Ability.TB.Lasers
    properties {
     iD=#44
     instance {
      _structType=&FortStateTreeStartAbilityTaskInstanceData
      actor=null
      abilityTag {gameplayTags [{tagName=Ability.TB.Lasers}]}
     }
     node {
      name=None
      _structType=&FortStateTreeStartAbilityTask
      bCanEditConsideredForCompletion=true
      bConsideredForCompletion=true
      bHasTaskCompletionDelegateDispatcher=false
      bTaskEnabled=true
      transitionHandlingPriority=Normal
      bindingsBatch {value=65535}
      executionRuntimeTemplateIndex {value=65535}
      instanceDataHandle {index=65535, source=None, stateHandle {index=65535}}
      instanceTemplateIndex {value=65535}
      outputBindingsBatch {value=65535}
     }
    }
   }]
  transitions []
 }
]
