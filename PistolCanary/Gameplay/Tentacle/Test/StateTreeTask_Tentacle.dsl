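; StateTree task Blueprint in the tentacle test folder, apparently a placeholder with no gameplay behavior.
; On both state enter and state exit it runs two nested IsValid checks on empty inputs.
; Nothing follows the inner check, and the transition parameter goes unused.
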(event EventEnterState
  (params (Transition :type State%20Tree%20Transition%20Result%20Structure%20%28by%20ref%29))
  (Utilities|IsValid 0
    (:Is%20Valid
      (Utilities|IsValid 0))))

(event EventExitState
  (params (Transition :type State%20Tree%20Transition%20Result%20Structure%20%28by%20ref%29))
  (Utilities|IsValid 0
    (:Is%20Valid
      (Utilities|IsValid 0))))
